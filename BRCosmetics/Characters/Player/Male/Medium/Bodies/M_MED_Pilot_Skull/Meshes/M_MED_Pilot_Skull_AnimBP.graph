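# Post-pose Anim Blueprint for the M_MED_Pilot_Skull player body cosmetic, adding secondary motion to its attached gear.
# It copies the parent mesh's pose, then rotates and simulates the front pouch, using a thigh planar limit.
# Left-thigh yaw, remapped each update, fades in both pouch steps.
# Weighted multi-bone constraints position the bottle, shoulder socket and two grenades, with half-strength AnimDynamics layered on the grenades.
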AnimBlueprint /BRCosmetics/Characters/Player/Male/Medium/Bodies/M_MED_Pilot_Skull/Meshes/M_MED_Pilot_Skull_AnimBP
target_skeleton=&/M_MED_Pilot_Skull_Skeleton
graphs [
 AnimGraph {nodes [
   n1 = Get thigh_left_alpha() -> (thigh_left_alpha:double)
   n2 = "Transform (Modify) Bone - Bone: dyn_front_pouch"(
    ComponentPose=Local To Component(LocalPose=Copy Pose From Mesh() {bUseAttachedParent=true})
    Alpha=@n1
    Translation="0,0,0"
    Rotation="90,0,0"
    Scale="1,1,1"
   ) {
    rotationMode=BMM_Additive
    rotationSpace=BCS_BoneSpace
    alphaBoolBlend {blendOption=Linear}
    boneToModify {boneName=dyn_front_pouch}
    rotation {pitch=90}
   }
   n3 = Anim Dynamics - Bone: dyn_front_pouch(ComponentPose=@n2, Alpha=@n1, ExternalForce="0,0,0") {
    angularDampingOverride=0.9999
    bGravityOverrideInSimSpace=true
    bOverrideAngularDamping=true
    simulationSpace=RootRelative
    alphaBoolBlend {blendOption=Linear}
    boundBone {boneName=dyn_front_pouch}
    physicsBodyDefinitions [{
      collisionType=CustomSphere
      sphereCollisionRadius=1
      boundBone {boneName=dyn_front_pouch}
      boxExtents=<17 17 17>
      constraintSetup {angularTargetAxis=AxisZ, twistAxis=AxisZ, angularLimitsMax=<0 30 0>, angularLimitsMin=<0 -30 0>, angularTarget=<0 0 0>}
      localJointOffset=<0 0 -17>
     }]
    planarLimits [{drivingBone {boneName=thigh_l}, planeTransform {location=<0 8 0>, rotation=<rot 0 0 90>}}]
   }
   n4 = "Constraint - dyn_bottle (PPPPPP)"!(
    ComponentPose=@n3
    Alpha=1
    ConstraintWeights_0=0.0101
    ConstraintWeights_1:float=0.1863
    ConstraintWeights_2:float=0.0578
    ConstraintWeights_3:float=0.1833
    ConstraintWeights_4:float=0.0079
    ConstraintWeights_5:float=0.5547
   ) -> (Pose) {
    alphaBoolBlend {blendOption=Linear}
    boneToModify {boneName=dyn_bottle}
    constraintSetup [
     {targetBone {boneName=clavicle_l}}
     {targetBone {boneName=spine_01}}
     {targetBone {boneName=spine_02}}
     {targetBone {boneName=spine_03}}
     {targetBone {boneName=clavicle_r}}
     {targetBone {boneName=pelvis}}
    ]
    constraintWeights [1,1,1,1,1,1]
   }
   n5 = "Constraint - dyn_shoulder_socket (PP)"!(ComponentPose=@n4, Alpha=1, ConstraintWeights_0=0.6, ConstraintWeights_1:float=0.4) -> (Pose) {
    alphaBoolBlend {blendOption=Linear}
    boneToModify {boneName=dyn_shoulder_socket}
    constraintSetup [{targetBone {boneName=clavicle_r}}, {targetBone {boneName=upperarm_twist_01_r}}]
    constraintWeights [1,1]
   }
   n6 = "Constraint - dyn_grenade_02 (PPPPPPP)"!(
    ComponentPose=@n5
    Alpha=1
    ConstraintWeights_0=0.0029
    ConstraintWeights_1:float=0.1516
    ConstraintWeights_2:float=0.0485
    ConstraintWeights_3:float=0.0597
    ConstraintWeights_4:float=0.0017
    ConstraintWeights_5:float=0.7353
    ConstraintWeights_6:float=0.0002
   ) -> (Pose) {
    alphaBoolBlend {blendOption=Linear}
    boneToModify {boneName=dyn_grenade_02}
    constraintSetup [
     {targetBone {boneName=clavicle_l}}
     {targetBone {boneName=spine_01}}
     {targetBone {boneName=spine_02}}
     {targetBone {boneName=spine_03}}
     {targetBone {boneName=spine_04}}
     {targetBone {boneName=pelvis}}
     {targetBone {boneName=thigh_twist_01_l}}
    ]
    constraintWeights [1,1,1,1,1,1,1]
   }
   n7 = "Constraint - dyn_grenade_01 (PPPPPPP)"!(
    ComponentPose=@n6
    Alpha=1
    ConstraintWeights_0=0.0015
    ConstraintWeights_1:float=0.1424
    ConstraintWeights_2:float=0.0731
    ConstraintWeights_3:float=0.0658
    ConstraintWeights_4:float=0.0017
    ConstraintWeights_5:float=0.7132
    ConstraintWeights_6:float=0.0024
   ) -> (Pose) {
    alphaBoolBlend {blendOption=Linear}
    boneToModify {boneName=dyn_grenade_01}
    constraintSetup [
     {targetBone {boneName=clavicle_l}}
     {targetBone {boneName=spine_01}}
     {targetBone {boneName=spine_02}}
     {targetBone {boneName=spine_03}}
     {targetBone {boneName=spine_04}}
     {targetBone {boneName=pelvis}}
     {targetBone {boneName=thigh_twist_01_l}}
    ]
    constraintWeights [1,1,1,1,1,1,1]
   }
   n8 = Anim Dynamics - Bone: dyn_grenade_02(ComponentPose=@n7, Alpha=0.5, ExternalForce="0,0,0") {
    angularDampingOverride=0.9999
    bGravityOverrideInSimSpace=true
    bOverrideAngularDamping=true
    gravityScale=0.35
    simulationSpace=RootRelative
    alphaBoolBlend {blendOption=Linear}
    boundBone {boneName=dyn_grenade_02}
    physicsBodyDefinitions [{
      sphereCollisionRadius=0
      boundBone {boneName=dyn_grenade_02}
      boxExtents=<20 50 50>
      constraintSetup {angularLimitsMax=<0 20 0>, angularLimitsMin=<0 -20 -40>, angularTarget=<0 0 0>}
      localJointOffset=<15 0 0>
     }]
   }
   n9 = Anim Dynamics - Bone: dyn_grenade_01(ComponentPose=@n8, Alpha=0.5, ExternalForce="0,0,0") {
    angularDampingOverride=0.9999
    bGravityOverrideInSimSpace=true
    bOverrideAngularDamping=true
    gravityScale=0.4
    simulationSpace=RootRelative
    alphaBoolBlend {blendOption=Linear}
    boundBone {boneName=dyn_grenade_01}
    physicsBodyDefinitions [{
      sphereCollisionRadius=0
      boundBone {boneName=dyn_grenade_01}
      boxExtents=<16 30 30>
      constraintSetup {angularLimitsMax=<0 20 0>, angularLimitsMin=<0 -20 -40>, angularTarget=<0 0 0>}
      localJointOffset=<10 0 0>
     }]
   }
   Output Pose(Result=Component To Local(ComponentPose=@n9)) {layerGroup=None}
  ]}
 EventGraph {nodes [
   Event BlueprintUpdateAnimation()
   n1 = BreakTransform(InTransform=GetDeltaTransformFromRefPose(self=GetPartSkeletalMeshComponent(), BoneName=thigh_l, BaseName=None))
   > Set thigh_left_alpha(thigh_left_alpha:double=MapRangeClamped(Value=BreakRotator(InRot=@n1.Rotation).Yaw, InRangeA=-40, InRangeB=-100)) -> (Output_Get:double)
  ]}
]
variables [thigh_left_alpha {type=number}]
