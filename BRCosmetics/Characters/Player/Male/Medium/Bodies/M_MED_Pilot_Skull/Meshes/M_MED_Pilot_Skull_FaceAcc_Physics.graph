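# Physics asset for the face accessory mesh of the male medium Pilot Skull character, covering a tube bone and nearby upper-body bones.
# A tilted box body on dyn_tube is constrained to a tiny head capsule with tight angular limits, slightly wider on swing 2.
# The head links to a spine_05 capsule with moderate limits.
# A left clavicle capsule also links to spine_05, with all rotation free and no right-side counterpart.
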PhysicsAsset /BRCosmetics/Characters/Player/Male/Medium/Bodies/M_MED_Pilot_Skull/Meshes/M_MED_Pilot_Skull_FaceAcc_Physics
bodies {
 clavicle_l [{shapeType=Capsule, length=20, radius=10, center=<12.8654 0 0>, rotation=<rot 79.9914 180 180>}]
 dyn_tube [{shapeType=Box, extentX=3.313, extentY=24.46, extentZ=34.8426, center=<-2.9125 0 -7.3148>, rotation=<rot 20 0 0>}]
 head [{shapeType=Capsule, length=1, radius=1}]
 spine_05 [{shapeType=Capsule, length=17.0558, radius=9.2727}]
}
constraints [
 {bone1Name=head, bone2Name=spine_05, swing1LimitDegrees=45, swing2LimitDegrees=45, twistLimitDegrees=45}
 {
  bone1Name=clavicle_l
  bone2Name=spine_05
  swing1LimitDegrees=45
  swing1Motion=Free
  swing2LimitDegrees=45
  swing2Motion=Free
  twistLimitDegrees=45
  twistMotion=Free
 }
 {bone1Name=dyn_tube, bone2Name=head, swing1LimitDegrees=5, swing2LimitDegrees=20, twistLimitDegrees=5}
]
