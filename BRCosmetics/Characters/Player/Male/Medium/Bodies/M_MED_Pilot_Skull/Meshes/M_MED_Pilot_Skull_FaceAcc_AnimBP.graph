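# Post-copy AnimBlueprint for the Pilot Skull face accessory, fitting its tanks and tube to the parent mesh's pose.
# Weighted constraints place back-tank pivots and head tube from spine, neck, head and clavicles; Look At and rotation constraints orient the tanks.
# Helmet-tank pitch drives arm_up_r, blending additive tube and tank offsets before RigidBody simulation.
# The Anim Dynamics node, debug print and other arm variable setters are disconnected.
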AnimBlueprint /BRCosmetics/Characters/Player/Male/Medium/Bodies/M_MED_Pilot_Skull/Meshes/M_MED_Pilot_Skull_FaceAcc_AnimBP
target_skeleton=&/M_MED_Pilot_Skull_FaceAcc_Skeleton
graphs [
 AnimGraph {nodes [
   Anim Dynamics - Bone: dyn_tube(Alpha=1, ExternalForce="0,0,0") {
    angularDampingOverride=1
    angularSpringConstant=1500
    bAngularSpring=true
    bGravityOverrideInSimSpace=true
    bOverrideAngularDamping=true
    gravityScale=0.35
    simulationSpace=RootRelative
    alphaBoolBlend {blendOption=Linear}
    boundBone {boneName=dyn_tube}
    physicsBodyDefinitions [{
      sphereCollisionRadius=0
      boundBone {boneName=dyn_tube}
      boxExtents=<10 20 20>
      constraintSetup {angularLimitsMax=<0 5 0>, angularLimitsMin=<0 -40 0>, angularTarget=<0 0 0>}
      localJointOffset=<-10 0 0>
     }]
   }
   n1 = "Constraint - dyn_back_tank_pivotA (PPPPP)"!(
    ComponentPose=Local To Component(LocalPose=Copy Pose From Mesh() {bUseAttachedParent=true})
    Alpha=1
    ConstraintWeights_0=0.1942
    ConstraintWeights_1:float=0.1486
    ConstraintWeights_2:float=0.074
    ConstraintWeights_3:float=0.0758
    ConstraintWeights_4:float=0.5073
   ) -> (Pose) {
    alphaBoolBlend {blendOption=Linear}
    boneToModify {boneName=dyn_back_tank_pivotA}
    constraintSetup [
     {targetBone {boneName=clavicle_r}}
     {targetBone {boneName=neck_02}}
     {targetBone {boneName=neck_01}}
     {targetBone {boneName=head}}
     {targetBone {boneName=spine_05}}
    ]
    constraintWeights [1,1,1,1,1]
   }
   n2 = "Constraint - dyn_back_tank_pivotB (PPP)"!(ComponentPose=@n1, Alpha=1, ConstraintWeights_0=0.0722, ConstraintWeights_1:float=0.3325, ConstraintWeights_2:float=0.5953) -> (Pose) {
    alphaBoolBlend {blendOption=Linear}
    boneToModify {boneName=dyn_back_tank_pivotB}
    constraintSetup [{targetBone {boneName=spine_03}}, {targetBone {boneName=spine_05}}, {targetBone {boneName=clavicle_l}}]
    constraintWeights [1,1,1]
   }
   n3 = Look At - Bone: dyn_back_tank_pivotA(ComponentPose=@n2, Alpha=1) {
    alphaBoolBlend {blendOption=Linear}
    boneToModify {boneName=dyn_back_tank_pivotA}
    lookAtLocation {x=0}
    lookAtTarget {boneReference {boneName=dyn_back_tank_pivotB}}
    lookAt_Axis {axis=<1 0>}
    lookUp_Axis {axis=<0 1>}
   }
   n4 = "Constraint - dyn_helmet_tank (R)"(ComponentPose=@n3, Alpha=1, ConstraintWeights_0=1) {
    alphaBoolBlend {blendOption=Linear}
    boneToModify {boneName=dyn_helmet_tank}
    constraintSetup [{transformType=Rotation, targetBone {boneName=dyn_back_tank_pivotA}}]
    constraintWeights [1]
   }
   n5 = "Constraint - dyn_head_tube (PPPPPP)"!(
    ComponentPose=@n4
    Alpha=1
    ConstraintWeights_0=0.035
    ConstraintWeights_1:float=0.0183
    ConstraintWeights_2:float=0.4641
    ConstraintWeights_3:float=0.3707
    ConstraintWeights_4:float=0.0797
    ConstraintWeights_5:float=0.0322
   ) -> (Pose) {
    alphaBoolBlend {blendOption=Linear}
    boneToModify {boneName=dyn_head_tube}
    constraintSetup [
     {targetBone {boneName=head}}
     {targetBone {boneName=spine_03}}
     {targetBone {boneName=spine_05}}
     {targetBone {boneName=clavicle_r}}
     {targetBone {boneName=neck_02}}
     {targetBone {boneName=neck_01}}
    ]
    constraintWeights [1,1,1,1,1,1]
   }
   n6 = Get arm_up_r() -> (arm_up_r:double)
   n7 = "Transform (Modify) Bone - Bone: dyn_head_tube"(ComponentPose=@n5, Alpha=@n6, Translation="2,-0.000,-0.000", Rotation="0,0,0", Scale="1,1,1") {
    translationMode=BMM_Additive
    translationSpace=BCS_BoneSpace
    alphaBoolBlend {blendOption=Linear}
    boneToModify {boneName=dyn_head_tube}
    translation {x=2}
   }
   n8 = "Transform (Modify) Bone - Bone: dyn_helmet_tank"(ComponentPose=@n7, Alpha=@n6, Translation="-1,-0.000,1", Rotation="0,10,0", Scale="1,1,1") {
    rotationMode=BMM_Additive
    rotationSpace=BCS_BoneSpace
    translationMode=BMM_Additive
    translationSpace=BCS_BoneSpace
    alphaBoolBlend {blendOption=Linear}
    boneToModify {boneName=dyn_helmet_tank}
    rotation {yaw=10}
    translation {x=-1, z=1}
   }
   n9 = RigidBody(ComponentPose=@n8) {
    bClampLinearTranslationLimitToRefPose=true
    bOverrideWorldGravity=true
    overridePhysicsAsset=&/M_MED_Pilot_Skull_FaceAcc_Physics
    alphaBoolBlend {blendOption=Linear}
    baseBoneRef {boneName=pelvis}
    componentAppliedLinearAccClamp=<3000 3000 3000>
    componentLinearAccScale=<1 1 1>
    componentLinearVelScale=<1 1 1>
    overrideWorldGravity {z=-980}
   }
   Output Pose(Result=Component To Local(ComponentPose=@n9)) {layerGroup=None}
   Get arm_up_more_r() -> (arm_up_more_r:double)
  ]}
 EventGraph {nodes [
   Event BlueprintUpdateAnimation()
   n1 = GetPartSkeletalMeshComponent()
   n2 = BreakTransform(InTransform=GetDeltaTransformFromRefPose(self=@n1, BoneName=dyn_helmet_tank))
   > Set arm_up_r(arm_up_r:double=MapRangeClamped(Value=BreakRotator(InRot=@n2.Rotation).Pitch, InRangeA=0, InRangeB=23)) -> (Output_Get:double)
   Reroute Node()
   n3 = BreakTransform(InTransform=GetDeltaTransformFromRefPose(self=@n1, BoneName=dyn_helmet_tank))
   > PrintString!(
    execute
    InString="To String (Rotator)"!(InRot:Rotator=@n3.Rotation) -> (ReturnValue:String)
    bPrintToScreen
    bPrintToLog
    TextColor
    Duration=0.03
    Key
   ) -> (then)
   Set arm_up_more_r(arm_up_more_r:double=MapRangeClamped(Value="Absolute (Float)"(A=0), InRangeA=85, InRangeB=125)) -> (Output_Get:double)
   > Set arm_back_r(arm_back_r:double=MapRangeClamped(Value=0, InRangeA=35, InRangeB=-20)) -> (Output_Get:double)
   > Set arm_side_lift_r(arm_side_lift_r:double=MapRangeClamped(Value="Absolute (Float)"(A=0), InRangeA=0, InRangeB=60)) -> (Output_Get:double)
   GetOwnerSkeletalMeshComponent()
   GetOwnerSkeletalMeshComponent()
   GetDeltaTransformFromRefPose(BoneName=dyn_helmet_tank)
   "To String (Vector)"!(InVec:Vector=@n3) -> (ReturnValue:String)
  ]}
]
variables [
 arm_back_r {type=number}
 arm_side_lift_r {type=number}
 arm_up_more_r {type=number}
 arm_up_r {type=number}
]
